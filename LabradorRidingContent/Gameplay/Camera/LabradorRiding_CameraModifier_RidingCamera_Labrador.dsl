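; A camera modifier for Labrador riding that anchors the player camera's height to the ridden mount.
; It walks from the camera owner's player controller to the controlled pawn's rider component and reads that component's active ridable.
; The view's Z becomes the ridable owner's actor Z plus a ZOffset variable, while X, Y, rotation and FOV pass through unchanged.
; The height is set directly, with no smoothing or validity checks in the graph.
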(fn BlueprintModifyCamera
  (params (DeltaTime :type Float%20%28single-precision%29) (ViewLocation :type Vector :default "0, 0, 0") (ViewRotation :type Rotator :default "0, 0, 0") (FOV :type Float%20%28single-precision%29))
  (return (Math|Vector|MakeVector (.x ViewLocation) (.y ViewLocation) (+ (.z (Transformation|GetActorLocation (Components|GetOwner (Variables|/Script/RidingCodeRuntime.RiderComponent|GetDoNotUse_ActiveRidable (Actor|GetComponentbyClass (Pawn|GetControlledPawn (Game|Player|GetOwningPlayerController (Variables|CameraModifier|GetCameraOwner))) "/LabradorRidingContent/Components/LabradorRiding_PlayerRiderComponent.LabradorRiding_PlayerRiderComponent_C"))))) (Variables|Default|GetZOffset))) ViewRotation FOV))
